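# Material function for the CupidHunter backpack that layers either a CD surface or an audio-reactive waveform onto incoming material attributes, chosen by the IsCD static switch.
# The CD branch sets metallic, anisotropy and a custom UV-derived tangent, with color from a time-scrolled curve atlas gradient.
# The other branch overrides emissive, building a stepped, grid-masked waveform from sine waves and scrolling noise, with amplitude from MF_AudioReactivity.
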MaterialFunction /BRCosmetics/Accessories/FORT_Backpacks/Backpack_F_MED_CupidHunter/Materials/MF_CupidHunter_Backpack
graph {nodes [
  n1 = GetMaterialAttributes(None=FunctionInput() {inputName=Material, inputType=FunctionInput_MaterialAttributes}) {attributeGetTypes [#1]}
  n2 = SetMaterialAttributes!(
   MaterialAttributes=NamedRerouteDeclaration(None=@n1.MaterialAttributes) {name=MA_IN, nodeColor=<rgba 0.875 0 1 1>}
   BaseColor=NamedRerouteUsage()
   Metallic=ScalarParameter() {defaultValue=1, group=CupidHunter Backpack, parameterName=CD_Metallic, sliderMax=1}
   Specular=ScalarParameter() {defaultValue=1, group=CupidHunter Backpack, parameterName=CD_Specular, sliderMax=1}
   Roughness=ScalarParameter() {defaultValue=0.5, group=CupidHunter Backpack, parameterName=CD_Roughness, sliderMax=1}
   EmissiveColor
   Tangent=NamedRerouteUsage()
   Anisotropy=ScalarParameter() {defaultValue=0.9, group=CupidHunter Backpack, parameterName=CD_Anisotropy}
  ) {attributeSetTypes [
    #2
    #3
    #4
    #5
    #6
    #7
    #8
   ]}
  n3 = NamedRerouteUsage()
  n4 = NamedRerouteUsage()
  FunctionOutput(None=StaticSwitchParameter(True=@n2, False=SetMaterialAttributes!(MaterialAttributes=@n3, EmissiveColor=@n4) {attributeSetTypes [#6]}) {parameterName=IsCD})
  NamedRerouteDeclaration(None=@n1.Normal) {name=Normal_IN, nodeColor=<rgba 0.1484 0 1 1>}
  n5 = Multiply(A=NamedRerouteUsage()) {constB=0.3}
  n6 = Time()
  n7 = Add(A=Multiply(B=NamedRerouteUsage()) {constA=0.5}, B=@n6)
  n8 = Multiply(A=@n6) {constB=-0.2}
  n9 = MaterialFunctionCall(Float2=Add(A=NamedRerouteUsage(), B=Constant2Vector() {g=-0.25})) {materialFunction=&BreakOutFloat2Components}
  n10 = TextureSample(UVs=MaterialFunctionCall(X=Add(A=@n8, B=@n9.R), Y=Constant()) {materialFunction=&MakeFloat2}.Result) {samplerType=SAMPLERTYPE_Grayscale, texture=/Game/Global/Textures/Default/Noise/LowResBlurredNoiseNormalized}
  n11 = NamedRerouteUsage()
  n12 = Multiply(A=Sine(None=Add(A=@n7, B=@n10.R)), B=LinearInterpolate(Alpha=@n11) {constA=0.1, constB=0.3}) {constB=0.15}
  n13 = LinearInterpolate(Alpha=@n11) {constA=0.26, constB=0}
  n14 = Multiply(A=Add(A=Add(A=@n12, B=@n13) {constB=0.25}, B=@n9.G), B=Step(Y=@n9.G) {constX=0.5})
  n15 = Multiply(A=Step(X=Add(A=@n5, B=@n14)) {constY=0.5}, B=NamedRerouteUsage())
  n16 = Step(Y=ComponentMask(None=NamedRerouteUsage()) {g=true}) {constX=0.99, constY=0.99}
  NamedRerouteDeclaration(None=Multiply(
    A=Multiply(A=@n15, B=@n16)
    B=VectorParameter() {group=CupidHunter Backpack, parameterName=WaveformColor, defaultValue=<rgba 2 0 1.8278 1>}.RGB
   )) {name=WaveformEmissive, nodeColor=<rgba 1 0 0.1875 1>}
  n17 = MaterialFunctionCall!(Offset, UV=TextureCoordinate() {coordinateIndex=1}, Tiling=Constant() {r=10}) {materialFunction=/Game/Global/Materials/MaterialGym/Materials/TechArtPatterns/MaterialFunctions/MF_GridUVs}
  NamedRerouteDeclaration(None=@n17.GridUV) {name=GridUV, nodeColor=<rgba 1 0 0.5859 1>}
  n18 = Step(Y=NamedRerouteDeclaration(None=@n17.GridUV_Tiling) {name=TilingUV, nodeColor=<rgba 0 1 0.8359 1>}) {constX=0.2, constY=0.9}
  NamedRerouteDeclaration(None=OneMinus(None=MaterialFunctionCall(f2=@n18) {materialFunction=&AddComponents}.f2)) {name=GridMask, nodeColor=<rgba 0.1953 0 1 1>}
  n19 = Multiply(A=Time(), B=ScalarParameter() {defaultValue=0.3, group=CupidHunter Backpack, parameterName=CD_RotateSpeed})
  n20 = ComponentMask(None=TextureCoordinate()) {r=true}
  n21 = CurveAtlasRowParameter(CurveTime=Frac(None=Add(A=@n19, B=@n20))) {
   atlas=/BRCosmetics/Accessories/FORT_Backpacks/Backpack_F_MED_CupidHunter/Textures/CA_CupidHunter
   curve=/BRCosmetics/Accessories/FORT_Backpacks/Backpack_F_MED_CupidHunter/Textures/C_CupidHunter
   group=CupidHunter Backpack
   parameterName=CD_Curve
  }
  NamedRerouteDeclaration(None=LinearInterpolate(A=@n21, B=VectorParameter() {group=CupidHunter Backpack, parameterName=CD_Color, defaultValue=<rgba 0.3281 0.3281 0.3281 1>}.RGB)) {name=CD_Color, nodeColor=<rgba 0.1953 1 0 1>}
  n22 = OneMinus(None=ComponentMask(None=TextureCoordinate()) {r=true})
  n23 = OneMinus(None=Distance(A=@n22, B=Constant() {r=0.5}))
  n24 = Constant() {r=1}
  NamedRerouteDeclaration(None=Power(Base=MaterialFunctionCall(X=@n22, Y=@n23, Z=@n24) {materialFunction=&MakeFloat3}.Result) {constExponent=0.4545}) {name=CD_Tangent, nodeColor=<rgba 1 0.2344 0 1>}
  n25 = MaterialFunctionCall() {materialFunction=/Game/Packages/Fortress_SharedMaterials/Base_Material_Functions/BaseCharacterShaderFunctions/Audio/MF_AudioReactivity}
  n26 = Constant() {r=8}
  NamedRerouteDeclaration(None=SmoothStep(Value=Power(Base=@n25.Amplitude, Exp=@n26) {constExponent=3})) {name=Amplitude, nodeColor=<rgba 0 1 0.4609 1>}
 ]}
